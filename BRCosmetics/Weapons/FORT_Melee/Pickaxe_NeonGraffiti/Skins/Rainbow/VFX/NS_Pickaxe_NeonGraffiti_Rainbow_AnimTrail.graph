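# Niagara swing trail for the Rainbow skin of the NeonGraffiti pickaxe, drawn as a ribbon following the weapon's animation.
# A single FN_AnimTrail emitter spawns short-lived ribbon particles through shared AnimTrail spawn, initialize and update modules, with user parameters for skeletal mesh and ribbon ID offset.
# A color curve cycles warm yellow, green, cyan, purple and pink while fading alpha, and a float curve narrows the ribbon.
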NiagaraSystem /BRCosmetics/Weapons/FORT_Melee/Pickaxe_NeonGraffiti/Skins/Rainbow/VFX/NS_Pickaxe_NeonGraffiti_Rainbow_AnimTrail
dependencies {
 usedDataInterfaces [&NiagaraDataInterfaceColorCurve, &NiagaraDataInterfaceCurve]
 usedDynamicInputs []
 usedModules [&SystemState]
}
emitters {FN_AnimTrail {
  emitterUpdateScript [
   &EmitterState(Life Cycle Mode=::NewEnumerator1, Loop Behavior=::NewEnumerator1, Loop Delay=0.1, Loop Duration=0.15, UseLoopDelay=true)
   &FN_SpawnAnimTrail_v2()
  ]
  particleSpawnScript [
   /Game/Effects/Niagara/Modules/AnimTrail/FN_InitializeAnimTrail/FN_InitializeAnimTrailBoneIndecies(Second Bone or Socket Index=1)
   /Game/Effects/Niagara/Modules/AnimTrail/FN_InitializeAnimTrail/FN_InitializeAnimTrail(AnimTrailWidth=0.7, Color=<rgba 2 2 2 1>, Lifetime=0.2)
  ]
  particleUpdateScript [
   &SolveForcesAndVelocity(Clamp Velocity=false, Limit Acceleration=false)
   &FN_UpdateAnimTrail_v2()
   &ScaleColor(
    Linear Color Curve=DataInterface() {propertyValues {
      AlphaCurve {keys [<at 0 1>,<at 0.2 0.7>,<at 1 0>]}
      BlueCurve {keys [<at 0.0742 0.1198>,<at 0.2833 0.125>,<at 0.4958 0.956>,<at 0.7033 1>,<at 0.909 0.4689>]}
      GreenCurve {keys [<at 0.0742 0.7359>,<at 0.2833 1>,<at 0.4958 1>,<at 0.7033 0.125>,<at 0.909 0.1615>]}
      RedCurve {keys [<at 0.0742 1>,<at 0.2833 0.1687>,<at 0.4958 0.1198>,<at 0.7033 0.4604>,<at 0.909 1>]}
     }}
    Scale Mode=::NewEnumerator2
   )
   &ScaleRibbonWidth(Ribbon Width Scale=&FloatFromCurve(FloatCurve=DataInterface() {propertyValues {Curve {keys [<at 0 1>,<at 1 0.5>]}}}))
  ]
  rendererClasses [&NiagaraRibbonRendererProperties]
  renderers [{rendererClass=&NiagaraRibbonRendererProperties}]
 }}
summary {
 systemName=NS_Pickaxe_NeonGraffiti_Rainbow_AnimTrail
 emitters [{emitterName=FN_AnimTrail, rendererClasses [&NiagaraRibbonRendererProperties]}]
 userVariables [
  User.RibbonIDOffset {type=NiagaraInt32}
  User.SkeletalMesh {
   type=NiagaraDataInterfaceSkeletalMesh
   defaultValue {
    struct=NiagaraExt_VariableValue_DataInterface
    value {dataInterface=&self:NiagaraDataInterfaceSkeletalMesh_0}
   }
  }
 ]
}
